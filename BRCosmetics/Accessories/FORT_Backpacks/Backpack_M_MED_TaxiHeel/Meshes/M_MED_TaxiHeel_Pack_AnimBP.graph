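# Animation Blueprint for the TaxiHeel backpack, posing the pack and driving a meter needle that reacts to the wearer's health and shield.
# Pack offsets follow spine bend and tactical-sprint jumps.
# The needle kicks one way on damage, the other on gains, then springs back, with noise-driven wobble in the frontend and AnimDynamics secondary motion.
# Being downed or carried blends to a sine-jittered, pegged needle pose that bypasses those layers.
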AnimBlueprint /BRCosmetics/Accessories/FORT_Backpacks/Backpack_M_MED_TaxiHeel/Meshes/M_MED_TaxiHeel_Pack_AnimBP
target_skeleton=&/M_MED_TaxiHeel_Pack_Skeleton
graphs [
 AnimGraph {nodes [
   n1 = "Transform (Modify) Bone - Bone: pack"!(
    ComponentPose=Local To Component(LocalPose=Use cached pose 'InputPoseCache'())
    bAlphaBoolEnabled:bool
    Translation=Get backpack_translation()
    Rotation=Get backpack_rotate()
    Scale
   ) -> (Pose) {
    alphaInputType=Bool
    rotationMode=BMM_Additive
    rotationSpace=BCS_BoneSpace
    translationMode=BMM_Additive
    translationSpace=BCS_BoneSpace
    alphaBoolBlend {blendInTime=0.5, blendOutTime=2}
    boneToModify {boneName=pack}
   }
   n2 = "Transform (Modify) Bone - Bone: pack"(ComponentPose=@n1, bAlphaBoolEnabled=Get Tactical Jumping() -> (Tactical Jumping:bool), Translation="0,-6,0", Rotation="0,0,0") {
    alphaInputType=Bool
    rotationSpace=BCS_BoneSpace
    translationMode=BMM_Additive
    translationSpace=BCS_BoneSpace
    alphaBoolBlend {blendInTime=0.1, blendOutTime=0.4}
    boneToModify {boneName=pack}
    translation {y=-6}
   }
   n3 = "Transform (Modify) Bone - Bone: base_meter_dial"!(
    ComponentPose=@n2
    bAlphaBoolEnabled:bool=True
    Translation="0,0,0"
    Rotation_Roll:float=0
    Rotation_Pitch:float=Get CurrentNeedleRot() -> (CurrentNeedleRot:double)
    Rotation_Yaw:float=0
    Scale="1,1,1"
   ) -> (Pose) {
    alphaInputType=Bool
    rotationMode=BMM_Replace
    rotationSpace=BCS_BoneSpace
    alphaBoolBlend {blendInTime=0.5, blendOutTime=2}
    boneToModify {boneName=base_meter_dial}
   }
   n4 = Anim Dynamics - Bone: dyn_meter_dial(ComponentPose=@n3, Alpha=1, ExternalForce="0,0,0") {
    angularDampingOverride=1
    angularSpringConstant=35000
    bAngularSpring=true
    bOverrideAngularDamping=true
    bUseGravityOverride=true
    gravityScale=0.35
    simulationSpace=RootRelative
    alphaScaleBiasClamp {bInterpResult=true, interpSpeedDecreasing=3, interpSpeedIncreasing=3}
    boundBone {boneName=dyn_meter_dial}
    componentAppliedLinearAccClamp=<3000 3000 3000>
    componentLinearAccScale=<0.3 0.3 0.3>
    componentLinearVelScale=<0.3 0.3 0.3>
    gravityOverride {z=-500}
    physicsBodyDefinitions [{
      boundBone {boneName=dyn_meter_dial}
      boxExtents=<30 30 30>
      constraintSetup {twistAxis=AxisY, angularLimitsMax=<0 15 0>, angularLimitsMin=<0 -15 0>, angularTarget=<1 0 0>}
      localJointOffset=<0 0 2>
     }]
   }
   n5 = "Transform (Modify) Bone - Bone: base_meter_dial"!(
    ComponentPose=Local To Component(LocalPose=Use cached pose 'InputPoseCache'())
    Alpha:float=1
    Translation="0,0,0"
    Rotation:Rotator=Get Needle DBNO() -> (Needle DBNO:Rotator)
    Scale="1,1,1"
   ) -> (Pose) {rotationMode=BMM_Additive, rotationSpace=BCS_BoneSpace, boneToModify {boneName=base_meter_dial}}
   Output Pose(Result=Blend Poses by bool(
     BlendPose_0=Component To Local(ComponentPose=@n5)
     BlendPose_1=Component To Local(ComponentPose=@n4)
     BlendTime_0=0.5
     BlendTime_1=0.5
     bActiveValue=Get Is DBNO/DBNO Carry() -> (Is DBNO/DBNO Carry:bool)
    )) {layerGroup=None}
   Save cached pose 'InputPoseCache'(Pose=Input Pose()) {cacheName=InputPoseCache}
  ]}
 EventGraph {nodes [
   Event BlueprintInitializeAnimation()
   > Set Skel Mesh(Skel Mesh=GetOwnerSkeletalMeshComponent())
   > n1 = Cast To FortPlayerPawn(Object=TryGetFortPawn())
   > n2 = Set Pawn(Pawn:Fort Player Pawn Object Reference=@n1) -> (Output_Get:Fort Player Pawn Object Reference)
   > Branch(Condition=OnLocalPlayersTeam!(TargetActor:Actor Object Reference=@n2) -> (Return Value:bool))
   > n3 = WatchPawnHealth!(execute:Exec, PlayerPawn:Fort Player Pawn Object Reference=@n2) -> (then:Exec, OnHitPointsChanged:Exec, Health:float=~, MaxHealth:float=~, Shield:float=~, MaxShield:float=~) {is=title K2Node_AsyncAction}
   > Set LastHealth(LastHealth:double=@n3) -> (Output_Get:double)
   > Set Last Shield(Last Shield:float=@n3.Shield) -> (Output_Get:float)
   n4 = Event BlueprintUpdateAnimation()
   > Set Delta Time X(Delta Time X=@n4.DeltaTimeX)
   > n5 = Sequence!(execute) -> (then_0, then_1, then_2:Exec)
   n6 = Branch(execute=@n5.then_2, Condition=Get bIsInFrontEnd())
   n7 = GetServerWorldTimeSeconds!() -> (ReturnValue:float)
   n8 = MapRangeClamped(Value="% (Float)"!(A:double=@n7, B:double=10) -> (ReturnValue:double), InRangeA=5, InRangeB=7)
   n9 = PerlinNoise1D!(Value:float=float * float(A=@n7, B=0.7) {is=title K2Node_PromotableOperator}) -> (ReturnValue:float)
   > n10 = FloatSpringInterp!(
    execute
    Current:float=float * float(A=float * float(A=@n8, B=@n9), B=90)
    Target:float=0
    SpringState:"Float Spring State Structure (by ref)"=Get Spring State() -> (Spring State:Float Spring State Structure)
    Stiffness:float=1
    CriticalDampingFactor:float=0.6
    DeltaTime:float=Get Delta Time X()
    Mass:float=400
    TargetVelocityAmount:float=0.5
    bClamp:bool=false
    MinValue:float=-80
    MaxValue:float=80
    bInitializeFromTarget:bool=false
   ) -> (then, ReturnValue:float)
   > Set CurrentNeedleRot(CurrentNeedleRot:double="Clamp (Float)"(Value=@n10, Min=-75, Max=75)) -> (Output_Get:double)
   n11 = WatchPawnHealth(execute=@n6.else, PlayerPawn=Get Pawn() -> (Pawn:Fort Player Pawn Object Reference)) -> (Health=~, MaxHealth=~, Shield=~, MaxShield=~)
   > Set Current Health(Current Health:float=@n11) -> (Output_Get:float)
   > Set Current Shield(Current Shield:float=@n11.Shield) -> (Output_Get:float)
   > n12 = Sequence!(execute) -> (then_0, then_1)
   n13 = Get LastHealth() -> (LastHealth:double)
   n14 = Get Current Health() -> (Current Health:float)
   n15 = Get Last Shield() -> (Last Shield:float)
   n16 = Get Current Shield() -> (Current Shield:float)
   > n17 = Branch(Condition=OR Boolean(A="float < float"(A=@n14, B=@n13), B="float < float"(A=@n16, B=@n15)))
   n18 = Get CurrentNeedleRot()
   > Set CurrentNeedleRot(CurrentNeedleRot="Clamp (Float)"(Value=float + float(A=@n18, B=30), Min=0, Max=80))
   > Set LastHealth(LastHealth=Get Current Health())
   > Set Last Shield(Last Shield=Get Current Shield())
   > Set Dynamics Alpha(Dynamics Alpha:float=1) -> (Output_Get:float)
   n19 = Get Last Shield()
   n20 = Get Current Shield()
   n21 = Get LastHealth()
   n22 = Get Current Health()
   n23 = Branch(execute=@n17.else, Condition=OR Boolean(A="float > float"(A=@n22, B=@n21), B="float > float"(A=@n20, B=@n19)))
   n24 = Get CurrentNeedleRot()
   > Set CurrentNeedleRot(CurrentNeedleRot="Clamp (Float)"(Value=float - float(A=@n24, B=30), Min=-80, Max=0))
   > Set LastHealth(LastHealth=Get Current Health())
   > Set Last Shield(Last Shield=Get Current Shield())
   > Set Dynamics Alpha(Dynamics Alpha=1)
   Set Last Shield(execute=@n23.else, Last Shield=Get Current Shield())
   > Set Dynamics Alpha(Dynamics Alpha=0)
   n25 = FloatSpringInterp(
    execute=@n12.then_1
    Current=Get CurrentNeedleRot()
    Target=0
    SpringState=Get Spring State()
    Stiffness=1
    CriticalDampingFactor=0.8
    DeltaTime=Get Delta Time X()
    Mass=1
    TargetVelocityAmount=0.5
    bClamp=false
    MinValue=-90
    MaxValue=90
    bInitializeFromTarget=false
   )
   > Set CurrentNeedleRot(CurrentNeedleRot=@n25)
   n26 = GetOwnerSkeletalMeshComponent()
   n27 = BreakRotator(InRot=BreakTransform(InTransform=GetDeltaTransformFromRefPose(self=@n26, BoneName=spine_05, BaseName=spine_01)).Rotation)
   Set backpack_rotate(execute=@n5.then_0, backpack_rotate=MakeRotator(Roll=MapRangeClamped(Value=@n27.Yaw, InRangeA=0, InRangeB=-30, OutRangeB=30)))
   > Set backpack_translation(backpack_translation=MakeVector(Y=MapRangeClamped(Value=@n27.Yaw, InRangeA=0, InRangeB=-30, OutRangeA=-2.5, OutRangeB=-8), Z=-3))
   n28 = Get bIsJumping()
   n29 = Get bIsFalling()
   > Set Tactical Jumping(Tactical Jumping:bool=AND Boolean(A=Get bIsTacticalSprinting(), B=OR Boolean(A=@n28, B=@n29))) -> (Output_Get:bool)
   n30 = GetGameTimeInSeconds()
   n31 = float * float(A="Sin (Radians)"(A=float * float(A=25, B=@n30)), B=1) {is=title K2Node_CommutativeAssociativeBinaryOperator}
   Set Needle DBNO(execute=@n5.then_1, Needle DBNO:Rotator=MakeRotator(Pitch=MapRangeClamped(Value=@n31, InRangeA=-1, InRangeB=1, OutRangeA=70, OutRangeB=85))) -> (Output_Get:Rotator)
   > Set Is DBNO/DBNO Carry(Is DBNO/DBNO Carry:bool=OR Boolean(A=Get bIsDBNO(), B=Get bIsDBNOCarried())) -> (Output_Get:bool)
  ]}
]
variables [
 backpack_rotate {type=Rotator}
 backpack_translation {type=Vector}
 Current Health {type=number}
 Current Shield {type=number}
 CurrentNeedleRot {type=number}
 CurrentRotation {type=Rotator}
 Delta Time X {type=number}
 Dynamics Alpha {type=number}
 Is DBNO/DBNO Carry {type=boolean}
 Last Shield {type=number}
 LastHealth {type=number}
 Needle DBNO {type=Rotator}
 NewVar {type=Quat, default=<0 0 0 1>}
 Pawn {type=FortPlayerPawn}
 RotationTarget {type=Rotator}
 Skel Mesh {type=SkeletalMeshComponent}
 Spring State {type=FloatSpringState, default {}}
 Spring State_0 {type=QuaternionSpringState, default {}}
 Tactical Jumping {type=boolean}
 Target {type=Quat, default=<0 0 0 1>}
]
